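# Physics asset for the F_MED_BlueBadass female player body cosmetic, covering three dangling tear gas props on the right thigh.
# Each dyn_tearGas bone (A, B, C) has a small capsule body, and thigh_r has a larger capsule.
# Each prop is constrained directly to the thigh with tight twist and swing1 limits and a wide swing2 limit, suggesting a hinge-like sway.
# No other body parts have physics bodies.
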PhysicsAsset /BRCosmetics/Characters/Player/Female/Medium/Bodies/F_MED_BlueBadass/Meshes/F_MED_Blue_Badass_Physics
bodies {
 dyn_tearGasA [{shapeType=Capsule, length=6.817, radius=3.0251, center=<3.2032 -4.2118 0.578>, rotation=<rot 0 90.7596 -87.7183>}]
 dyn_tearGasB [{shapeType=Capsule, length=6.817, radius=3.0224, center=<2.6602 -3.7521 -0.7865>, rotation=<rot -0.0175 -92.5585 -89.1463>}]
 dyn_tearGasC [{shapeType=Capsule, length=6.817, radius=3.023, center=<3.0014 2.8787 0.6667>, rotation=<rot -0.0104 87.7297 -89.3811>}]
 thigh_r [{shapeType=Capsule, length=11.5951, radius=8.2375, center=<23.3387 0.0001 0.8114>, rotation=<rot 0 -89.9999 -79.9675>}]
}
constraints [
 {bone1Name=dyn_tearGasC, bone2Name=thigh_r, swing1LimitDegrees=2, swing2LimitDegrees=40, twistLimitDegrees=2}
 {bone1Name=dyn_tearGasB, bone2Name=thigh_r, swing1LimitDegrees=2, swing2LimitDegrees=40, twistLimitDegrees=2}
 {bone1Name=dyn_tearGasA, bone2Name=thigh_r, swing1LimitDegrees=2, swing2LimitDegrees=40, twistLimitDegrees=2}
]
